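# Translucent unlit Niagara sprite material that draws a radiation texture for the SilentNovel Vane pickaxe effects.
# A hard sphere mask splits the texture's red channel into center and outer elements, which the ElementsLerp parameter blends between.
# Particle color tints the result into an emissive adjustment function, and particle alpha scales it for opacity.
# The default shading path depth-fades opacity, while mobile skips the fade.
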Material /BRCosmetics/Weapons/FORT_Melee/Pickaxe_SilentNovel_Vane/FX/M_SilentNovel_Vane_Radioactive
graph {
 properties {bUsedWithNiagaraSprites=true, blendMode=BLEND_Translucent, pixelDepthOffsetMode=PDOM_Legacy, shadingModel=MSM_Unlit, usedShadingModels=Unlit}
 nodes [
  n1 = TextureSample() {texture=&/T_Radiation_SilentNovel_Vane}
  n2 = Saturate(None=SphereMask(A=TextureCoordinate(), B=Constant() {r=0.5}, Radius=Constant() {r=0.15}, Hardness=Constant() {r=1}))
  n3 = LinearInterpolate(A=Multiply(A=@n1.R, B=@n2), B=Multiply(A=OneMinus(None=@n2), B=@n1.R), Alpha=ScalarParameter() {parameterName=ElementsLerp}) {constAlpha=0.15}
  MaterialFunctionCall(Emissive=Multiply(A=ParticleColor().RGB, B=@n3)) {materialFunction=&MF_FX_EmissiveAdjustments}
  n4 = Multiply(A=ParticleColor().A, B=@n3)
  ShadingPathSwitch(Default=DepthFade(Opacity=@n4, FadeDistance=ScalarParameter() {defaultValue=30, parameterName=FadeDistance}), Mobile=@n4) {desc=Migrated from FeatureLevelSwitch}
 ]
}
