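# Niagara swing trail for the SilentNovel Vane pickaxe, drawing green animation-trail ribbons during melee swings.
# Three ribbon emitters (SoftGlow, ToonGlow, ToonGlowScreen) share anim-trail spawn, initialize and update functions, differing in width, brightness and alpha curves.
# SoftGlow picks width sign and ribbon-width curve by SwingSectionIndex; both toon layers drive dynamic material parameters.
# All three tint by a time-of-day color multiplier and scale alpha by a system fade.
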NiagaraSystem /BRCosmetics/Weapons/FORT_Melee/Pickaxe_SilentNovel_Vane/FX/NS_Pickaxe_SilentNovel_Vane_Trail
dependencies {
 usedDataInterfaces [&NiagaraDataInterfaceCurve, &NiagaraDataInterfaceSocketReader, &NiagaraDataInterfaceColorCurve]
 usedDynamicInputs [&ColorFromCurve]
 usedModules [
  &self:SystemSpawnScript.SystemScriptSource.SystemScriptGraph.NiagaraNodeAssignment_1.SetVariables_#1
  &self:SystemSpawnScript.SystemScriptSource.SystemScriptGraph.NiagaraNodeAssignment_2.SetVariables_#2
  &SystemState
  &self:SystemSpawnScript.SystemScriptSource.SystemScriptGraph.NiagaraNodeAssignment_0.SetVariables_#3
  &self:SystemSpawnScript.SystemScriptSource.SystemScriptGraph.NiagaraNodeAssignment_3.SetVariables_#4
 ]
}
emitters {
 SoftGlow {
  emitterUpdateScript [
   &EmitterState()
   &FN_SpawnAnimTrail()
  ]
  particleSpawnScript [
   &FN_InitializeAnimTrailBoneIndecies()
   &FN_InitializeAnimTrail(
    AnimTrailWidth=&MergeHelper_FloatSwitch(0=1, 1=-1, 2=1, 3=-1, Index=$User.SwingSectionIndex, Num=::NewEnumerator2)
    Color=&MakeLinearColorFromVectorAndFloat("Vector (RGB)"=&MakeVectorFromLinearColorRGB(Color=<rgba 0.4626 1 0.2291 1>), "Float (Alpha)"=1)
    Lifetime=0.22
   )
  ]
  particleUpdateScript [
   &FN_UpdateAnimTrail() {module=FN_UpdateAnimTrail001}
   &ScaleColor(
    Scale Alpha=&Multiply_Float(
     A=&Multiply_Float(
      A=&MakeFloatFromLinearColor(LinearColor=$System.TODColorMultiplier, Channel:ENiagaraFloatFromLinearColorOptions=::NewEnumerator3)
      B=&FloatFromCurve(
       FloatCurve=DataInterface() {propertyValues {Curve {
          defaultValue=1
          keys [{interpMode=RCIM_Cubic, tangentMode=RCTM_User, time=0.2, value=1}, {interpMode=RCIM_Cubic, tangentMode=RCTM_User, time=0.66, value=0}]
         }}}
       Scale Curve=0.17
      )
     )
     B=$System.AlphaFade
    )
    Scale RGB=&MakeVectorFromLinearColorRGB(Color=$System.TODColorMultiplier)
    ScaleA=true
    ScaleRGB=true
   )
   &ScaleRibbonWidth(Ribbon Width Scale=&MergeHelper_FloatSwitch(
     0=&FloatFromCurve(FloatCurve=DataInterface() {propertyValues {Curve {keys [<at 0 0.7 RCIM_Cubic>, {arriveTangent=0.3, leaveTangent=0.3, time=0.2, value=1}, <at 1 1>]}}})
     1=&FloatFromCurve(FloatCurve=DataInterface() {propertyValues {Curve {keys [
          <at 0 0.7 RCIM_Cubic>
          {arriveTangent=-0.3416, interpMode=RCIM_Cubic, leaveTangent=-0.3416, tangentMode=RCTM_User, time=0.2, value=1}
          <at 1 1>
         ]}}})
     2=&FloatFromCurve(FloatCurve=DataInterface() {propertyValues {Curve {keys [
          <at 0 0.7 RCIM_Cubic>
          {arriveTangent=-0.3416, interpMode=RCIM_Cubic, leaveTangent=-0.3416, tangentMode=RCTM_User, time=0.2, value=1}
          <at 1 1>
         ]}}})
     3=&FloatFromCurve(FloatCurve=DataInterface() {propertyValues {Curve {keys [<at 0 0.7 RCIM_Cubic>, {arriveTangent=0.3, interpMode=RCIM_Cubic, leaveTangent=0.3, time=0.15, value=1}, <at 1 1>]}}})
     Index=$User.SwingSectionIndex
     Num=::NewEnumerator2
    ))
  ]
  rendererClasses [&NiagaraRibbonRendererProperties]
  renderers [{rendererClass=&NiagaraRibbonRendererProperties}]
 }
 ToonGlow {
  emitterUpdateScript [
   &EmitterState()
   &FN_SpawnAnimTrail()
  ]
  particleSpawnScript [
   &FN_InitializeAnimTrailBoneIndecies()
   &FN_InitializeAnimTrail(
    AnimTrailWidth=1.7
    Color=&MakeLinearColorFromVectorAndFloat("Vector (RGB)"=&MakeVectorFromLinearColorRGB(Color=<rgba 0.6917 2.6 0.3792 1>), "Float (Alpha)"=1)
    Lifetime=0.22
   )
  ]
  particleUpdateScript [
   &FN_UpdateAnimTrail() {module=FN_UpdateAnimTrail001}
   &DynamicMaterialParameters(
    Index 0 Param 1=0.7
    Index 0 Param 3=&Multiply_Float(
     A=0.3
     B=&FloatFromCurve(FloatCurve=DataInterface() {propertyValues {Curve {keys [{arriveTangent=1.7806, interpMode=RCIM_Cubic, leaveTangent=1.7806, tangentMode=RCTM_User, time=0, value=0}, <at 1 1>]}}})
    )
    Module.Index0Float0Write=true
    Module.Index0Float1Write=true
    Module.Index0Float2Write=true
    Module.Index0Float3Write=true
   )
   &ScaleColor(
    Scale Alpha=&Multiply_Float(
     A=&Multiply_Float(
      A=&MakeFloatFromLinearColor(LinearColor=$System.TODColorMultiplier, Channel:ENiagaraFloatFromLinearColorOptions=::NewEnumerator3)
      B=&FloatFromCurve(FloatCurve=DataInterface() {propertyValues {Curve {
          defaultValue=1
          keys [
           <at 0 0.6614 RCIM_Cubic>
           {arriveTangent=-0.2698, interpMode=RCIM_Cubic, leaveTangent=-0.2698, tangentMode=RCTM_User, time=0.3454, value=0.7838}
           {interpMode=RCIM_Cubic, tangentMode=RCTM_User, time=0.66, value=0}
          ]
         }}})
     )
     B=$System.AlphaFade
    )
    Scale RGB=&MakeVectorFromLinearColorRGB(Color=$System.TODColorMultiplier)
    ScaleA=true
    ScaleRGB=true
   )
  ]
  rendererClasses [&NiagaraRibbonRendererProperties]
  renderers [{rendererClass=&NiagaraRibbonRendererProperties}]
 }
 ToonGlowScreen {
  emitterUpdateScript [
   &EmitterState()
   &FN_SpawnAnimTrail()
  ]
  particleSpawnScript [
   &FN_InitializeAnimTrailBoneIndecies()
   &FN_InitializeAnimTrail(
    AnimTrailWidth=0.4
    Color=&MakeLinearColorFromVectorAndFloat("Vector (RGB)"=&MakeVectorFromLinearColorRGB(Color=<rgba 0.6917 2.6 0.3792 1>), "Float (Alpha)"=1)
    Lifetime=0.22
   )
  ]
  particleUpdateScript [
   &FN_UpdateAnimTrail() {module=FN_UpdateAnimTrail001}
   &DynamicMaterialParameters(
    Index 0 Param 1=0.7
    Index 0 Param 3=&Multiply_Float(
     A=0.3
     B=&FloatFromCurve(FloatCurve=DataInterface() {propertyValues {Curve {keys [{arriveTangent=1.7806, interpMode=RCIM_Cubic, leaveTangent=1.7806, tangentMode=RCTM_User, time=0, value=0}, <at 1 1>]}}})
    )
    Module.Index0Float0Write=true
    Module.Index0Float1Write=true
    Module.Index0Float2Write=true
    Module.Index0Float3Write=true
   )
   &ScaleColor(
    Scale Alpha=&Multiply_Float(
     A=&Multiply_Float(
      A=&MakeFloatFromLinearColor(LinearColor=$System.TODColorMultiplier, Channel:ENiagaraFloatFromLinearColorOptions=::NewEnumerator3)
      B=&FloatFromCurve(FloatCurve=DataInterface() {propertyValues {Curve {
          defaultValue=1
          keys [
           <at 0 0.6614 RCIM_Cubic>
           {arriveTangent=-0.2698, interpMode=RCIM_Cubic, leaveTangent=-0.2698, tangentMode=RCTM_User, time=0.3454, value=0.7838}
           {interpMode=RCIM_Cubic, tangentMode=RCTM_User, time=0.66, value=0}
          ]
         }}})
     )
     B=$System.AlphaFade
    )
    Scale RGB=&MakeVectorFromLinearColorRGB(Color=$System.TODColorMultiplier)
    ScaleA=true
    ScaleRGB=true
   )
  ]
  rendererClasses [&NiagaraRibbonRendererProperties]
  renderers [{rendererClass=&NiagaraRibbonRendererProperties}]
 }
}
summary {
 systemName=NS_Pickaxe_SilentNovel_Vane_Trail
 emitters [
  {emitterName=SoftGlow, rendererClasses [&NiagaraRibbonRendererProperties]}
  {emitterName=ToonGlow, rendererClasses [&NiagaraRibbonRendererProperties]}
  {emitterName=ToonGlowScreen, rendererClasses [&NiagaraRibbonRendererProperties]}
 ]
 userVariables [
  User.bIsFrontEnd {type=NiagaraBool}
  User.bIsFrontEndPreview {type=NiagaraBool}
  User.RibbonIDOffset {type=NiagaraInt32}
  User.SwingSectionIndex {type=NiagaraInt32}
 ]
}
